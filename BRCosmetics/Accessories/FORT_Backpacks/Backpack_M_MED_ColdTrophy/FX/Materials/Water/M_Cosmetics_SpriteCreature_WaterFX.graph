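# Translucent subsurface water material for Niagara mesh particles in the ColdTrophy backpack effects, with a black base and its color carried by emissive.
# Screen-space UVs anchored to the object and scaled by depth feed panning bubble normals and a noise-warped inner glow.
# Fresnel rim, sun glints, scattering and time-of-day scaling shape the lighting; depth, camera-distance and particle alpha fade opacity.
# Niagara dynamic parameters drive vertex distortion, bend and droplet inflation, alongside sine-wave offset math.
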Material /BRCosmetics/Accessories/FORT_Backpacks/Backpack_M_MED_ColdTrophy/FX/Materials/Water/M_Cosmetics_SpriteCreature_WaterFX
graph {
 properties {
  bAutomaticallySetUsageInEditor=false
  bUseMaterialAttributes=true
  bUsedWithNiagaraMeshParticles=true
  blendMode=BLEND_Translucent
  shadingModel=MSM_Subsurface
  usedShadingModels=Subsurface
 }
 nodes [
  n1 = ObjectPositionWS()
  n2 = TransformPosition(None=@n1) {transformSourceType=TRANSFORMPOSSOURCE_World, transformType=TRANSFORMPOSSOURCE_View}
  ComponentMask(None=@n2) {b=true, g=true}
  Constant() {r=0.4}
  Constant() {r=0.6}
  Constant() {r=-0.25}
  n3 = Constant() {r=1}
  n4 = Constant() {r=0.5}
  n5 = Constant()
  n6 = LinearInterpolate(
   A=VectorParameter() {parameterName=BaseColorA, defaultValue=<rgba 0.0524 0.2356 0.4375 1>}.RGB
   B=Multiply(
    A=MaterialFunctionCall(Value_Morning=@n3, Value_Day=@n3, Value_Evening=@n4, Value_Night=@n5) {materialFunction=&Scalar_Time_of_Day}.Result
    B=VectorParameter() {parameterName=BaseColorA_Scattered, defaultValue=<rgba 0 1 2 1>}.RGB
   )
   Alpha=NamedRerouteUsage()
  )
  n7 = Divide(A=ComponentMask(None=NamedRerouteUsage()) {b=true}) {constB=55}
  n8 = Constant() {r=0.35}
  LinearInterpolate(
   A=VectorParameter() {parameterName=BaseColorB, defaultValue=<rgba 0 0.1674 0.3438 1>}.RGB
   B=LinearInterpolate(A=@n6, B=VectorParameter() {parameterName=BaseColorC, defaultValue=<rgba 0 0.4727 1 1>}.RGB, Alpha=Step(Y=@n7, X=@n8) {constX=0.4})
   Alpha=Distance(A=@n7, B=@n8)
  )
  n9 = WorldPosition() {worldPositionShaderOffset=WPT_ExcludeAllShaderOffsets}
  n10 = ScalarParameter() {defaultValue=0.013, parameterName=WorldWaveWPOFrequency}
  n11 = Subtract(A=Frac(None=Multiply(A=@n9.XYZ, B=@n10) {constB=0.0025})) {constB=0.5}
  n12 = MaterialFunctionCall(Float3=Abs(None=Multiply(A=@n11) {constB=2})) {materialFunction=&BreakOutFloat3Components}
  n13 = Add(A=Add(A=@n12.R, B=@n12.R), B=@n12.B)
  n14 = Multiply(A=Time()) {constB=0.2}
  n15 = Sine(None=Subtract(A=Subtract(A=@n13) {constB=2}, B=Frac(None=@n14)))
  n16 = Multiply(A=Multiply(A=Add(A=@n15)) {constB=0.5}, B=ScalarParameter() {defaultValue=2, parameterName=WPOWorldScale})
  n17 = MaterialFunctionCall(Float3=Multiply(A=LocalPosition() {localOrigin=InstancePreSkinning}.XYZ, B=ScalarParameter() {defaultValue=0.05, parameterName=LocalWaveWPOFrequency}) {constB=0.02}) {materialFunction=&BreakOutFloat3Components}
  n18 = Add(A=@n17.R, B=@n17.R)
  n19 = Subtract(A=Time(), B=ScalarParameter() {parameterName=WPO_random_seed})
  n20 = Subtract(A=Subtract(A=Add(A=@n18, B=@n17.B)) {constB=2}, B=Frac(None=Multiply(A=@n19) {constB=0.75}))
  n21 = Multiply(A=Add(A=Sine(None=@n20))) {constB=0.5}
  Multiply(A=Multiply(A=VertexNormalWS(), B=Add(A=@n16, B=@n21) {constB=0}), B=ScalarParameter() {defaultValue=0.25, parameterName=WPO_FinalScale})
  NamedRerouteDeclaration(None=VertexInterpolator(VS=LocalPosition() {localOrigin=InstancePreSkinning}.XYZ).PS) {name=PreskinnedPosition, nodeColor=<rgba 1 0.2344 0 1>}
  n22 = NamedRerouteUsage()
  n23 = VectorParameter() {parameterName=SSC, defaultValue=<rgba 0 0.974 0.974 1>}
  n24 = Saturate(None=DotProduct(
    A=AtmosphericLightVector()
    B=MaterialFunctionCall() {materialFunction=&CameraDirectionVector}.Result
   ))
  NamedRerouteDeclaration(None=LinearInterpolate(B=Multiply(A=@n22, B=@n23.RGB), Alpha=NamedRerouteDeclaration(None=@n24) {name=SunMask, nodeColor=<rgba 0.8047 0 1 1>})) {name=SSC, nodeColor=<rgba 0 0.4141 1 1>}
  n25 = Constant()
  n26 = ScalarParameter() {defaultValue=0.3, parameterName=InnerGlowNoiseSpeed}
  n27 = TextureSample(UVs=Panner(Coordinate=NamedRerouteUsage(), Speed=AppendVector(A=@n25, B=@n26)) {speedY=0.2}) {texture=/Game/Environments/Helios/Props/Quest/Quest_GoldenFleeceStatuette_A/Textures/T_GoldenFleeceStatuette_CellNoise}
  n28 = ComponentMask(None=Multiply(A=ConstantBiasScale(None=@n27.RGB) {bias=-0.5, scale=2}, B=ScalarParameter() {defaultValue=0.05, parameterName=InnerGlowNoiseAmount})) {b=true, r=true}
  n29 = SphereMask(
   A=Add(A=@n28, B=NamedRerouteUsage())
   B=Constant2Vector() {g=0.1, r=0.2}
   Radius=ScalarParameter() {defaultValue=0.3, parameterName=InnerGlowRadius}
   Hardness=ScalarParameter() {defaultValue=-0.5, parameterName=InnerGlowHardness}
  ) {attenuationRadius=12, hardnessPercent=55}
  NamedRerouteDeclaration(None=@n29) {name=InnerGlow, nodeColor=<rgba 0.1719 0 1 1>}
  n30 = Clamp(
   None=ParticleRandom()
   Min=ScalarParameter() {defaultValue=0.6, parameterName=Param_3}
   Max=ScalarParameter() {defaultValue=0.6, parameterName=Param_4}
  ) {maxDefault=0}
  n31 = TextureCoordinate()
  n32 = Multiply(A=ComponentMask(None=@n31) {g=true}, B=ComponentMask(None=@n31) {r=true})
  n33 = DynamicParameter() {parameterIndex=2, defaultValue=<rgba 1 0.5 0 1>, paramNames [GradiantNumber, DistortionSpeed, EmissiveMult_Final, Param4]}
  n34 = Add(A=Multiply(A=Add(A=@n30, B=@n32), B=@n33.GradiantNumber), B=Multiply(A=@n33.DistortionSpeed, B=Time()))
  n35 = DynamicParameter() {parameterIndex=3, paramNames [X-WPOAmmount, Y-WPOAmmount, Z-WPOAmmount, GlobalWPODistortionStrength]}
  n36 = ComponentMask(None=MaterialFunctionCall(R=@n35.X-WPOAmmount, G=@n35.X-WPOAmmount, B=@n35.X-WPOAmmount) {materialFunction=&AppendMany}.RGB) {b=true, g=true, r=true}
  n37 = ConstantBiasScale(None=Sine(None=Multiply(A=@n34, B=@n36)))
  n38 = Multiply(A=ComponentMask(None=@n37) {r=true}, B=ComponentMask(None=@n37) {g=true})
  n39 = ScalarParameter() {defaultValue=1, parameterName=Param_2}
  n40 = Multiply(
   A=Multiply(
    A=VertexNormalWS()
    B=DynamicParameter() {parameterIndex=3, paramNames [X-WPOAmmount, Y-WPOAmmount, Z-WPOAmmount, GlobalWPODistortionStrength]}.GlobalWPODistortionStrength
   )
   B=Add(A=Power(Base=@n38, Exp=@n39), B=ComponentMask(None=@n37) {b=true})
  )
  n41 = DynamicParameter() {desc=Parameters set by Niagara, parameterIndex=1, paramNames [X,Y,Z,WPOIntensity]}
  n42 = MaterialFunctionCall(X=@n41.X, Y=@n41.X, Z=@n41.X) {materialFunction=&MakeFloat3}
  n43 = ScalarParameter() {defaultValue=-0.25}
  n44 = ComponentMask(None=TextureCoordinate()) {g=true}
  n45 = DynamicParameter() {defaultValue=<rgba 0 1 3 1>, paramNames [DropletPower, DropletInflate, WPO Bend Power, WPOPulseAmount]}
  n46 = Power(Base=@n44, Exp=@n45.WPO Bend Power)
  n47 = Multiply(A=Saturate(None=LinearInterpolate(A=@n43, Alpha=@n46) {constA=-0.5}), B=@n41.WPOIntensity)
  n48 = LinearInterpolate(A=@n40, B=Add(A=Multiply(A=@n42.Result, B=@n47), B=@n40), Alpha=ScalarParameter() {parameterName=UsePendulumWPO?})
  n49 = OneMinus(None=ComponentMask(None=TextureCoordinate()) {g=true})
  n50 = DynamicParameter() {defaultValue=<rgba 0 1 3 1>, paramNames [DropletPower, DropletInflate, WPO Bend Power, WPOPulseAmount]}
  n51 = Multiply(A=VertexNormalWS(), B=Multiply(A=Power(Base=@n49, Exp=@n50.DropletPower) {constExponent=1}, B=@n50.DropletInflate))
  NamedRerouteDeclaration(None=LinearInterpolate(A=@n48, B=Add(A=@n48, B=@n51), Alpha=ScalarParameter() {parameterName=UseWPOSpeedScale?})) {name=VFXWPO, nodeColor=<rgba 0 1 0.5781 1>}
  n52 = Constant2Vector()
  n53 = ScreenPosition()
  n54 = MaterialFunctionCall(Float2=ViewSize()) {materialFunction=&BreakOutFloat2Components}
  n55 = AppendVector(A=Constant() {r=1}, B=Divide(A=@n54.G, B=@n54.G))
  n56 = Multiply(A=ConstantBiasScale(None=Add(A=@n52, B=@n53.ViewportUV)) {bias=-0.5, scale=2}, B=@n55)
  n57 = Multiply(
   A=Constant2Vector() {g=-0.5, r=0.5}
   B=MaterialFunctionCall(World Position=@n1) {materialFunction=&TransformToClipSpace}.Clip Space XY
  ) {constA=1}
  n58 = ScalarParameter() {defaultValue=6, parameterName=Screen Space Scale}
  n59 = MaterialFunctionCall() {materialFunction=&ObjectScale}
  n60 = MaterialFunctionCall(
   UVs=Subtract(A=ConstantBiasScale(None=@n56), B=Multiply(A=@n55, B=@n57))
   Texture Scale=Divide(A=Multiply(A=@n58, B=@n59.Scale X), B=ComponentMask(None=@n2) {b=true}) {constA=100}
  ) {materialFunction=&ScaleUVsByCenter}
  n61 = Constant2Vector() {g=0.3, r=0.4}
  NamedRerouteDeclaration(None=Multiply(A=Multiply(A=@n60.UVs, B=@n61), B=MaterialFunctionCall() {materialFunction=/Engine/Functions/Engine_MaterialFunctions02/Utility/FOV}."FOV ")) {name=ScreenUV, nodeColor=<rgba 1 0 0.1172 1>}
  n62 = NamedRerouteUsage()
  n63 = Constant()
  n64 = MaterialFunctionCall(
   Normal Vector=Add(A=AppendVector(A=@n62, B=@n63), B=VertexNormalWS())
   Power=ScalarParameter() {defaultValue=3, parameterName=OutGlowFresnelPow}
  ) {materialFunction=&Fresnel_Function}
  NamedRerouteDeclaration(None=Multiply(A=@n64.Result, B=VectorParameter() {parameterName=OutGlowFresnelColor, defaultValue=<rgba 0 0.9323 0.6506 1>}.RGB)) {name=OuterGlow, nodeColor=<rgba 1 0.5156 0 1>}
  NamedRerouteDeclaration(None=@n28) {name=InnerGlowNoise, nodeColor=<rgba 0.3125 0 1 1>}
  n65 = DotProduct(A=AtmosphericLightVector(), B=PixelNormalWS())
  n66 = Multiply(A=SmoothStep(Value=@n65) {constMin=0.96})
  n67 = Multiply(A=SmoothStep(Value=@n65) {constMax=1.2, constMin=0.83})
  n68 = LinearInterpolate(
   A=LinearInterpolate(A=@n6, B=VectorParameter() {parameterName=InnerGlowColor, defaultValue=<rgba 0 0.52 0.6979 1>}.RGB, Alpha=@n29)
   B=VectorParameter() {parameterName=InnerScatteringColor, defaultValue=<rgba 0.023 0.057 0.276 1>}.RGB
   Alpha=NamedRerouteUsage()
  )
  n69 = NamedRerouteUsage()
  n70 = Multiply(
   A=Add(A=Add(A=@n66, B=@n67), B=Add(A=@n68, B=@n69))
   B=MaterialFunctionCall(
    Value_Morning=ScalarParameter() {defaultValue=1, parameterName=TOD_Morning}
    Value_Day=ScalarParameter() {defaultValue=1, parameterName=TOD_Day}
    Value_Evening=ScalarParameter() {defaultValue=1, parameterName=TOD_Evening}
    Value_Night=ScalarParameter() {defaultValue=1, parameterName=TOD_Night}
   ) {materialFunction=&Scalar_Time_of_Day}.Result
  )
  n71 = DynamicParameter() {parameterIndex=2, defaultValue=<rgba 1 0.5 0 1>, paramNames [GradiantNumber, DistortionSpeed, EmissiveMult_Final, Param4]}
  n72 = Add(A=ScalarParameter() {defaultValue=1, parameterName=EmissiveMult_Final}, B=@n71.EmissiveMult_Final)
  NamedRerouteDeclaration(None=EyeAdaptationInverse(LightValueInput=Multiply(A=@n70, B=@n72), AlphaInput=ScalarParameter() {defaultValue=1, parameterName=EyeAdaptionInverse}).EyeAdaptationInverse) {name=Emissive, nodeColor=<rgba 0.7344 1 0 1>}
  NamedRerouteDeclaration() {name=WPO, nodeColor=<rgba 0.9453 0 1 1>}
  n73 = AtmosphericLightVector()
  n74 = PixelNormalWS()
  n75 = DotProduct(A=@n73, B=@n74)
  n76 = Subtract(A=Multiply(A=Multiply(A=@n75) {constB=2}, B=@n74), B=@n73)
  n77 = MaterialFunctionCall() {materialFunction=&CameraDirectionVector}
  n78 = Max(A=DotProduct(A=Normalize(VectorInput=@n76), B=Multiply(A=@n77.Result) {constB=-1})) {constB=0}
  n79 = ScalarParameter() {defaultValue=2, parameterName=InnerScatteringPower}
  NamedRerouteDeclaration(None=Multiply(A=Power(Base=@n78, Exp=@n79), B=ScalarParameter() {defaultValue=0.7, parameterName=InnerScatteringScale}) {constB=0.4}) {name=InnerScatter, nodeColor=<rgba 1 0 0.1406 1>}
  n80 = Constant3Vector() {constant=<rgba 0 0 1 0>}
  n81 = NamedRerouteUsage()
  n82 = Multiply(A=NamedRerouteUsage()) {constB=10}
  n83 = Add(A=@n81, B=Multiply(A=Saturate(None=@n82), B=ScalarParameter() {defaultValue=-0.05, parameterName=BubbleEdgeWarp}))
  n84 = BumpOffset(
   Coordinate=Multiply(A=Panner(Coordinate=@n83) {speedY=0.15}, B=ScalarParameter() {defaultValue=1.5, parameterName=BubbleTilingR}) {constB=3}
   Height=ScalarParameter() {defaultValue=0.05, parameterName=BubbleOffsetR}
  ) {heightRatio=0.1}
  n85 = TextureObjectParameter() {
   group=Slurp
   parameterName=BubbleMask
   samplerType=SAMPLERTYPE_Masks
   texture=/Game/Characters/Player/Male/Large/Bodies/M_LRG_Slurp_Monster/Textures/T_SlurpBubble
  }
  n86 = TextureSample(UVs=@n84, Tex=@n85)
  n87 = TextureObjectParameter() {
   group=Slurp
   parameterName=BubbleNormal
   samplerType=SAMPLERTYPE_Normal
   texture=/Game/Characters/Player/Male/Large/Bodies/M_LRG_Slurp_Monster/Textures/T_SlurpBubble_N
  }
  n88 = Multiply(A=Panner(Coordinate=@n81) {speedY=0.1}, B=ScalarParameter() {defaultValue=1, parameterName=BubbleTilingG}) {constB=3}
  n89 = TextureSample(UVs=@n88, Tex=@n85)
  n90 = Add(A=Multiply(A=@n86.R, B=TextureSample(UVs=@n84, Tex=@n87).RGB), B=Multiply(A=@n89.G, B=TextureSample(UVs=@n88, Tex=@n87).RGB))
  n91 = Add(A=@n86.R, B=@n89.G)
  n92 = Multiply(A=LinearInterpolate(Alpha=NamedRerouteUsage())) {constB=10}
  NamedRerouteDeclaration(None=MaterialFunctionCall(Normal=LinearInterpolate(A=@n80, B=@n90, Alpha=@n91), Flatness=Saturate(None=@n92)) {materialFunction=&FlattenNormal}.Result) {name=BubbleNormal, nodeColor=<rgba 0 1 0.7891 1>}
  NamedRerouteDeclaration(None=@n64.Result) {name=OuterFresnel, nodeColor=<rgba 0 0.1797 1 1>}
  NamedRerouteUsage()
  OneMinus(None=@n44)
  n93 = PixelDepth()
  OneMinus(None=Saturate(None=Divide(A=@n93.R) {constB=100}))
  PixelNormalWS()
  ScalarParameter() {defaultValue=1, parameterName=Param_1}
  n94 = Constant()
  n95 = DepthFade(FadeDistance=ScalarParameter() {defaultValue=2, parameterName=DepthFadeDistance})
  n96 = PixelDepth()
  n97 = Constant()
  n98 = ScalarParameter() {parameterName=CameraDistanceFade}
  n99 = Power(Base=OneMinus(None=SphereMask(A=@n96.R, B=@n97, Radius=@n98) {hardnessPercent=10})) {constExponent=20}
  SetMaterialAttributes!(
   MaterialAttributes
   Base Color=@n94
   Metallic=@n94
   Specular=Constant() {r=1}
   Roughness=Constant() {r=0.25}
   Normal=NamedRerouteUsage()
   Emissive Color=NamedRerouteUsage()
   Opacity=Multiply(A=Multiply(A=@n95, B=@n99), B=ParticleColor().A)
   Subsurface Color=NamedRerouteUsage()
   World Position Offset=NamedRerouteUsage()
  ) {attributeSetTypes [
    #1
    #2
    #3
    #4
    #5
    #6
    #7
    #8
    #9
   ]}
  n100 = TextureObjectParameter() {
   group=Volume Textures
   parameterName=Displacement Volume Tex
   samplerType=SAMPLERTYPE_LinearColor
   texture=/Game/Effects/Explosion_Update/Textures/VT_Voronoi_Seamless
  }
  TextureSample(UVs=Constant3Vector() {constant=<rgba 1 1 1 0>}, Tex=@n100) {samplerType=SAMPLERTYPE_LinearColor, texture=&VT_Voronoi_Seamless}
 ]
}
